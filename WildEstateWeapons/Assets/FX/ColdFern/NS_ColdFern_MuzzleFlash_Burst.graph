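# Niagara muzzle flash burst for the ColdFern weapon, built from short-lived one-shot emitters.
# Mesh-and-sprite core flashes with random mirroring and system-wide random scale fire twice, 0.05s apart, as do speed-stretched sparks, alongside a brief orange light.
# Toon smoke takes its tint from a time-of-day color input.
# Four shockwave particles fan around the facing axis by partition index, which also picks per-particle material parameters from float arrays.
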NiagaraSystem /WildEstateWeapons/Assets/FX/ColdFern/NS_ColdFern_MuzzleFlash_Burst
dependencies {
 usedDataInterfaces [
  &NiagaraDataInterfaceCurve
  &NiagaraDataInterfaceVectorCurve
  &NiagaraDataInterfaceVector2DCurve
  &NiagaraDataInterfaceArrayFloat
 ]
 usedDynamicInputs [&Divide_Float, &SimulationPosition]
 usedModules [
  &self:SystemRandomScale
  &SystemState
  &self:SetSystemBoolsfromBitwise
  &self:SystemSpawnScript.SystemScriptSource.SystemScriptGraph.NiagaraNodeAssignment_0.SetVariables_#1
 ]
}
emitters {
 FN_Toon_Shockwave {
  emitterUpdateScript [
   &EmitterState() {module=EmitterLifeCycle}
   &SpawnBurst_Instantaneous(Spawn Count=4, Use Spawn Probability=false, UseLoopCountLimit=false)
  ]
  particleSpawnScript [
   /SetVariables(Particles.FX_Scale=0.6) {isSetParameters=true}
   /SetVariables(Particles.FacingVector:Vector3f=<1 0 0>) {isSetParameters=true}
   &InitializeParticle(
    Lifetime=0.6
    Mesh Scale=&Multiply_VectorByFloat(Vector=&RandomRangeVector(Minimum=<0.6 1 1>, Maximum=<1 2 1>), Float=$Particles.FX_Scale)
    Mesh Scale Mode=::NewEnumerator4
    UsePositionOffset=false
    Write Material Random=true
   ) {module=InitializeParticle001}
   &PartitionParticles(Number Of Partitions=$Engine.ExecutionCount)
   &ShapeLocation(
    Cylinder Distribution=::NewEnumerator2
    Cylinder Height=0
    Cylinder Radius=&Multiply_Float(A=$Particles.FX_Scale, B=10)
    MaskSpawnGroup=false
    Non Uniform Scale=<1 1 0>
    Shape Primitive=::NewEnumerator1
    U Position=$Particles.PartitionParticles.NormalizedPartition
    UseOwnerScale=true
    V Position=0
   )
   &InitialMeshOrientation(
    Mesh Orientation Mode=::NewEnumerator6
    Orientation Quaternion:Quat4f=&MultiplyQuaternion(
     Quaternion A=&MakeQuaternion(Quaternion From=0, Axis=<0 0 -1>, Angle=45, From=?, To=?, Angle Type=::NewEnumerator0, Normalize Vectors=false)
     Quaternion B=&MultiplyQuaternion(
      Quaternion A=&MakeQuaternion(
       Quaternion From=0
       Axis=<0 0 1>
       Angle=$Particles.PartitionParticles.NormalizedPartition
       From=?
       To=?
       Angle Type=::NewEnumerator1
       Normalize Vectors=true
      )
      Quaternion B=&MakeQuaternion(Quaternion From=0, Axis=<0 1 0>, Angle=-30, From=?, To=?, Angle Type=::NewEnumerator0, Normalize Vectors=true)
     )
    )
    Use Rotation Vector=false
   )
   /SetVariables(
    Particles.MeshOrientation=&MultiplyQuaternion(
     Quaternion A=&MakeQuaternion(
      Quaternion From=3
      Axis=<1 0 0>
      Angle=0
      From=$Particles.FacingVector
      To=<0 0 1>
      Angle Type=::NewEnumerator0
      Coordinate Space=::Local
      Normalize Vectors=true
     )
     Quaternion B=$Particles.Initial.MeshOrientation
    )
    Particles.Position=&AddVectorToPosition(
     Position=/Niagara/DynamicInputs/Helpers/SimulationPosition()
     Vector=&RotateVectorByQuaternion(
      Vector To Rotate=<0 0 0>
      Quaternion=&MakeQuaternion(
       Quaternion From=3
       Axis=<1 0 0>
       Angle=0
       From=$Particles.FacingVector
       To=<0 0 1>
       Angle Type=::NewEnumerator0
       Coordinate Space=::Local
       Normalize Vectors=false
      )
     )
    )
   ) {isSetParameters=true}
   &InheritVelocity(Speed Limit=true, UseSpeedThreshold=false)
  ]
  particleUpdateScript [
   &ParticleState()
   &ScaleMeshSize(Scale Factor=&MakeVector(
     X=&FloatFromCurve(FloatCurve=DataInterface() {propertyValues {Curve {
         defaultValue=0
         keys [
          {arriveTangent=2.6809, interpMode=RCIM_Cubic, leaveTangent=2.6809, tangentMode=RCTM_User, time=0, value=0}
          {arriveTangent=0.8569, interpMode=RCIM_Cubic, leaveTangent=0.8569, tangentMode=RCTM_User, time=0.5219, value=0.7989}
          {arriveTangent=0.1983, interpMode=RCIM_Cubic, leaveTangent=0.1983, tangentMode=RCTM_User, time=1, value=1}
         ]
        }}})
     Y=&FloatFromCurve(FloatCurve=DataInterface() {propertyValues {Curve {
         defaultValue=0
         keys [
          {arriveTangent=0.066, interpMode=RCIM_Cubic, leaveTangent=0.066, tangentMode=RCTM_User, time=0, value=0}
          {interpMode=RCIM_Cubic, tangentMode=RCTM_User, time=1, value=1}
         ]
        }}})
     Z=1
    ))
   &SolveForcesAndVelocity(Clamp Velocity=false, Limit Acceleration=false)
   &DynamicMaterialParameters(
    Index 0 Param 1=&SelectFloatFromArray(
     Direct Array Index=$Particles.PartitionParticles.Partition
     Float Selection Array=DataInterface() {propertyValues {FloatData [-0.1,-0.2,0,-0.1,0]}}
    )
    Index 0 Param 2=&SelectFloatFromArray(
     Direct Array Index=$Particles.PartitionParticles.Partition
     Float Selection Array=DataInterface() {propertyValues {FloatData [-0.2,-0.1,-0.1,-0.2,-0.1]}}
    )
    Index 0 Param 3=&SelectFloatFromArray(Direct Array Index=$Particles.PartitionParticles.Partition, Float Selection Array=DataInterface() {propertyValues {FloatData [0,0,1,1,0]}})
    Index 0 Param 4=&FloatFromCurve(
     FloatCurve=DataInterface() {propertyValues {Curve {
        defaultValue=0
        keys [{interpMode=RCIM_Cubic, tangentMode=RCTM_User, time=0, value=-1}, {interpMode=RCIM_Cubic, tangentMode=RCTM_User, time=1, value=1}]
       }}}
     CurveIndex=&Add_Float(A=$Particles.NormalizedAge, B=&RandomRangeFloat(Minimum=0, Maximum=0.2))
    )
    Index 1 Param 1=&SelectFloatFromArray(Direct Array Index=$Particles.PartitionParticles.Partition, Float Selection Array=DataInterface() {propertyValues {FloatData [-0.1,0,0,0]}})
    Index 1 Param 2=&SelectFloatFromArray(
     Array Sampling Mode=::NewEnumerator0
     Direct Array Index=?
     Float Selection Array=DataInterface() {propertyValues {FloatData [-0.2,0,0,0]}}
     Random Seed=0
    )
    Module.Index0Float0Write=true
    Module.Index0Float1Write=true
    Module.Index0Float2Write=true
    Module.Index0Float3Write=true
    Module.Index1Float0Write=false
    Module.Index1Float1Write=false
    Module.Index1Float2Write=true
    Module.Index1Float3Write=true
    Write Parameter Index 1=::NewEnumerator1
   )
   &ScaleColor(
    Scale Alpha=&FloatFromCurve(
     FloatCurve=DataInterface() {propertyValues {Curve {
        defaultValue=1
        keys [
         {interpMode=RCIM_Cubic, tangentMode=RCTM_User, time=0, value=0}
         {interpMode=RCIM_Cubic, tangentMode=RCTM_User, time=0.5, value=1}
         {tangentMode=RCTM_User, time=1, value=0}
        ]
       }}}
     Scale Curve=0.3
    )
    ScaleA=true
    ScaleRGB=true
   )
  ]
  rendererClasses [&NiagaraSpriteRendererProperties, &NiagaraMeshRendererProperties]
  renderers [
   {rendererClass=&NiagaraSpriteRendererProperties}
   {rendererClass=&NiagaraMeshRendererProperties, rendererIndex=1}
  ]
 }
 Lighting {
  emitterUpdateScript [
   &EmitterState()
   &SpawnBurst_Instantaneous(Use Spawn Probability=false, UseLoopCountLimit=false)
  ]
  particleSpawnScript [&InitializeParticle(
    Lifetime=0.1
    Position Offset=<10 0 0>
    Sprite Size Mode=::NewEnumerator3
    Uniform Sprite Size=15
    UsePositionOffset=true
    Write Material Random=true
   )]
  particleUpdateScript [
   &ParticleState()
   &FN_Light_Attributes(
    Color=<rgba 1 0.5161 0 1>
    Create Color Attribute=true
    Create Exponent Attribute=true
    Create Radius Attribute=true
    Exponent=2
    Radius=100
    Scale Color=10
   )
  ]
  rendererClasses [&NiagaraLightRendererProperties]
  renderers [{rendererClass=&NiagaraLightRendererProperties, rendererIndex=0}]
 }
 Muzzle_Core_Elements {
  emitterUpdateScript [
   &EmitterState()
   &SpawnBurst_Instantaneous(Use Spawn Probability=false, UseLoopCountLimit=false)
   &SpawnBurst_Instantaneous(Spawn Group=1, Spawn Time=0.05, Use Spawn Probability=false, UseLoopCountLimit=false) {module=SpawnBurst_Instantaneous001}
  ]
  particleSpawnScript [
   &InitializeParticle(
    Color=<rgba 1 0.758 0.5 1>
    Lifetime=&RandomRangeFloat(Minimum=0.1125, Maximum=0.15)
    Mesh Scale=&Multiply_Vector(
     A=&Multiply_VectorByFloat(Vector=<1.5 0.75 0.75>, Float=$System.RandomScale)
     B=&MakeVector(
      X=1
      Y=&UniformAOrBFloat(A=-1, Evaluation Type=::NewEnumerator0, Recalculate Random Each Loop=true, Override Seed=false)
      Z=&UniformAOrBFloat(A=-1, Evaluation Type=::NewEnumerator0, Recalculate Random Each Loop=true, Override Seed=false)
     )
    )
    Mesh Scale Mode=::NewEnumerator4
    Position Offset=&MaskVectorBySpawnGroup(Masked Vector=<-2 3 0>, Passthrough Vector=<-2 -3 0>)
    Sprite Rotation Mode=::NewEnumerator2
    Sprite Size=&Vector2DFromFloat(Value=&Multiply_Float(A=75, B=$System.RandomScale))
    Sprite Size Mode=::NewEnumerator4
    UsePositionOffset=true
    Write Material Random=true
   )
   &InitialMeshOrientation(
    Mesh Orientation Mode=::NewEnumerator2
    Rotation=&RandomRangeVector(Minimum=<0 0 0>, Maximum=<1 0 0>)
    Use Rotation Vector=true
   )
   &CameraOffset(Camera Offset Amount=1)
   &SpriteFacingAndAlignment(
    Sprite Facing=&MaskVectorBySpawnGroup(Masked Vector=<1 0 0>, Passthrough Vector=<1 1 1>)
    WriteAlignment=true
    WriteFacing=true
   )
   /SetVariables(Particles.SpriteCameraOffset=5) {isSetParameters=true}
  ]
  particleUpdateScript [
   &ParticleState()
   &ScaleSpriteSize(
    Scale Factor=&Vector2DFromFloat(Value=&FloatFromCurve(FloatCurve=DataInterface() {propertyValues {Curve {keys [{tangentMode=RCTM_User, time=0, value=0}, {tangentMode=RCTM_User, time=1, value=1}]}}}))
    Scale Sprite Size Mode=::NewEnumerator2
   )
   &ScaleMeshSize(Scale Factor=&VectorFromCurve(VectorCurve=DataInterface() {propertyValues {XCurve {keys [<at 0 0>,<at 1 1>]}, YCurve {keys [<at 0 0.25>,<at 1 1.25>]}, ZCurve {keys [<at 0 0.25>,<at 1 1.25>]}}}))
   &ScaleColor(
    Scale Alpha=&RandomRangeFloat(Minimum=0.5, Maximum=1)
    Scale RGB=&MaskVectorBySpawnGroup(
     Masked Vector=&VectorFromFloat(Value=&FloatFromCurve(FloatCurve=DataInterface() {propertyValues {Curve {keys [{arriveTangent=-42.9916, interpMode=RCIM_Cubic, leaveTangent=-42.9916, tangentMode=RCTM_User, time=0, value=5}, <at 0.15 1 RCIM_Cubic>]}}}))
     Passthrough Vector=<1 1 1>
    )
    ScaleA=true
    ScaleRGB=true
   )
   &DynamicMaterialParameters(
    Index 0 Param 1=&FloatFromCurve(
     FloatCurve=DataInterface() {propertyValues {Curve {keys [
         {arriveTangent=2.9073, interpMode=RCIM_Cubic, leaveTangent=2.9073, leaveTangentWeight=0.3077, tangentMode=RCTM_User, time=0.35, value=0}
         {arriveTangentWeight=0.8354, interpMode=RCIM_Cubic, tangentMode=RCTM_User, time=1, value=1}
        ]}}}
     Scale Curve=0.75
    )
    Index 0 Param 2=&FloatFromCurve(FloatCurve=DataInterface() {propertyValues {Curve {keys [
         {interpMode=RCIM_Cubic, leaveTangentWeight=0.3627, tangentMode=RCTM_User, time=0.4, value=1}
         {arriveTangent=-1.799, arriveTangentWeight=1.0345, interpMode=RCIM_Cubic, leaveTangent=-1.799, tangentMode=RCTM_User, time=1, value=0}
        ]}}})
    Index 0 Param 3=&MakeCustomFloatFromBool(Bool=&RandomBool(), True Float=1, False Float=0)
    Module.Index0Param0Write=true
    Module.Index0Param1Write=true
    Module.Index0Param2Write=true
    Module.Index0Param3Write=false
   )
   /SetVariables(Particles.SpritePOsition:NiagaraPosition=&AddVectorToPosition(Position=$Particles.Position, Vector=<2 0 0>)) {isSetParameters=true}
   &ScaleSpriteSize(
    Scale Factor=&MakeVector2D(
     X=&MakeFloatFromVector(Vector=$Engine.Owner.Scale, Channel=::NewEnumerator0)
     Y=&MakeFloatFromVector(Vector=$Engine.Owner.Scale, Channel=::NewEnumerator1)
    )
    Scale Sprite Size Mode=::NewEnumerator2
   ) {module=ScaleSpriteSize001}
  ]
  rendererClasses [&NiagaraMeshRendererProperties, &NiagaraSpriteRendererProperties]
  renderers [
   {rendererClass=&NiagaraMeshRendererProperties, rendererIndex=0}
   {rendererClass=&NiagaraSpriteRendererProperties, rendererIndex=1}
  ]
 }
 SparksLong001 {
  emitterUpdateScript [
   &EmitterState()
   &SpawnBurst_Instantaneous(
    Spawn Count=&RandomRangeInt(Minimum=15, Maximum=21)
    Use Spawn Probability=false
    UseLoopCountLimit=false
   )
   &SpawnBurst_Instantaneous(
    Spawn Count=&RandomRangeInt(Minimum=15, Maximum=21)
    Spawn Time=0.05
    Use Spawn Probability=false
    UseLoopCountLimit=false
   ) {module=SpawnBurst_Instantaneous001}
  ]
  particleSpawnScript [
   &InitializeParticle(
    Color=<rgba 4 2.0643 0 1>
    Lifetime Max=0.24
    Lifetime Min=0.14
    Lifetime Mode=::NewEnumerator1
    Position Offset=<-5 0 0>
    Position Offset Coordinate Space=::Local
    Sprite Size Max=<3 1.5>
    Sprite Size Min=<2 1.2>
    Sprite Size Mode=::NewEnumerator2
    UsePositionOffset=true
    Write Material Random=true
   ) {module=InitializeParticle001}
   &ShapeLocation(
    MaskSpawnGroup=false
    Non Uniform Scale=<2 1 1>
    Override Seed=false
    Rotation Mode=::NewEnumerator2
    Sphere Radius=5
    UseOwnerScale=true
    Yaw / Pitch / Roll=<-0.25 0 0>
   )
   &AddVelocityInCone(Module.Use Velocity Falloff On Cone Axis=false, Velocity Strength=1600)
   &AddVelocity(Velocity=&RandomRangeVector(Minimum=<150 0 0>, Maximum=<200 0 0>)) {module=AddVelocity001}
  ]
  particleUpdateScript [
   &ParticleState()
   &Drag(Drag=4, Use Linear Drag=true, Use Rotational Drag=true)
   &SolveForcesAndVelocity(Clamp Velocity=false, Limit Acceleration=false)
   &ScaleColor(
    Scale Alpha=&FloatFromCurve(FloatCurve=DataInterface() {propertyValues {Curve {keys [<at 0 1 RCIM_Cubic>,<at 1 0 RCIM_Cubic>]}}})
    ScaleA=true
    ScaleRGB=true
   )
   &ScaleSpriteSizeBySpeed(Max Scale Factor=<0.5 25>, Velocity Threshold=700)
  ]
  rendererClasses [&NiagaraSpriteRendererProperties]
  renderers [{rendererClass=&NiagaraSpriteRendererProperties}]
 }
 Toon_Impact_Smoke001 {
  emitterUpdateScript [
   &EmitterState()
   &SpawnBurst_Instantaneous(Spawn Count=5, Use Spawn Probability=false, UseLoopCountLimit=false)
  ]
  particleSpawnScript [
   &InitializeParticle(
    Color=&FNTimeOfDayColor(
     Night Color=<rgba 0.2149 0.3102 0.842 1>
     Morning Color=<rgba 0.7078 0.5363 0.5148 1>
     Day Color=<rgba 0.8096 0.8096 0.8096 1>
     Evening Color=<rgba 0.6076 0.4758 0.7031 1>
    )
    Lifetime Max=0.3
    Lifetime Min=0.2
    Lifetime Mode=::NewEnumerator1
    Sprite Rotation Angle=&RandomRangeFloat(Minimum=-45, Maximum=45)
    Sprite Rotation Mode=::NewEnumerator0
    Sprite Size Mode=::NewEnumerator1
    Uniform Sprite Size Max=50
    Uniform Sprite Size Min=25
    UsePositionOffset=false
    Write Material Random=true
   ) {module=InitializeParticle001}
   &AddVelocityInCone(
    Cone Angle=60
    Module.Use Velocity Falloff On Cone Axis=false
    Velocity Strength=&UniformRangedFloat(Minimum=100, Maximum=300)
   )
   &CameraOffset(Camera Offset Amount=5)
  ]
  particleUpdateScript [
   &ParticleState()
   &ScaleSpriteSize(Uniform Curve Sprite Scale=DataInterface() {propertyValues {Curve {
       defaultValue=0
       keys [{interpMode=RCIM_Cubic, tangentMode=RCTM_User, time=0, value=0.4}, {interpMode=RCIM_Cubic, tangentMode=RCTM_User, time=1, value=1}]
      }}})
   &DynamicMaterialParameters(
    Index 0 Param 1=&FloatFromCurve(FloatCurve=DataInterface() {propertyValues {Curve {keys [
         {arriveTangent=0.5561, interpMode=RCIM_Cubic, leaveTangent=0.5561, tangentMode=RCTM_User, time=0, value=0.5}
         {arriveTangent=0.9173, interpMode=RCIM_Cubic, leaveTangent=0.9173, tangentMode=RCTM_User, time=1, value=1}
        ]}}})
    Index 0 Param 2=&SelectFloatFromArray(
     Array Sampling Mode=::NewEnumerator0
     Direct Array Index=?
     Float Selection Array=DataInterface() {propertyValues {FloatData [0,1]}}
     Random Seed=0
    )
    Module.Index0Float0Write=true
    Module.Index0Float1Write=true
    Module.Index0Float2Write=true
    Module.Index0Float3Write=true
   ) {module=DynamicMaterialParameters001}
   &Drag(Drag=2, Use Linear Drag=true, Use Rotational Drag=true)
   &SolveForcesAndVelocity(Clamp Velocity=false, Limit Acceleration=false)
   &ScaleColor(
    Scale Alpha=&FloatFromCurve(FloatCurve=DataInterface() {propertyValues {Curve {keys [<at 0 1>,<at 0.8091 0.9932>,<at 1 0>]}}})
    ScaleA=true
    ScaleRGB=true
   )
  ]
  rendererClasses [&NiagaraSpriteRendererProperties]
  renderers [{rendererClass=&NiagaraSpriteRendererProperties}]
 }
}
summary {
 systemName=NS_ColdFern_MuzzleFlash_Burst
 emitters [
  {
   emitterName=Muzzle_Core_Elements
   rendererClasses [&NiagaraMeshRendererProperties, &NiagaraSpriteRendererProperties]
  }
  {emitterName=Lighting, rendererClasses [&NiagaraLightRendererProperties]}
  {emitterName=SparksLong001, rendererClasses [&NiagaraSpriteRendererProperties]}
  {emitterName=Toon_Impact_Smoke001, rendererClasses [&NiagaraSpriteRendererProperties]}
  {
   emitterName=FN_Toon_Shockwave
   rendererClasses [&NiagaraSpriteRendererProperties, &NiagaraMeshRendererProperties]
  }
 ]
 userVariables [
  User.AmmoType {type=NiagaraInt32, defaultValue=1}
  User.BurstLoops {type=NiagaraInt32, defaultValue=1}
  User.FiringRate {type=NiagaraFloat, defaultValue=0.1, description=""}
  User.GasCapsuleScale {type=Vector2f, description="", defaultValue {struct=Vector2f, value=<1 1>}}
  User.MuzzleColorHueControl {type=NiagaraFloat, defaultValue=0, description=""}
  User.MuzzleCoreCapScale {type=NiagaraFloat, defaultValue=1, description=""}
  User.MuzzleCoreScale {type=Vector3f, description="", defaultValue {struct=Vector3f, value=<1 1 1>}}
  User.MuzzleDecalAlpha {type=NiagaraFloat, defaultValue=1, description=""}
  User.MuzzleDecalOffset {type=Vector3f, description="", defaultValue {struct=Vector3f, value=<2 0 0>}}
  User.MuzzleDecalSize {type=NiagaraFloat, defaultValue=20, description=""}
  User.MuzzleModsBitSet {type=NiagaraInt32}
  User.SmokeEmissiveAmount {type=NiagaraFloat, defaultValue=1, description=""}
  User.SmokeInitialAlpha {type=NiagaraFloat, defaultValue=1, description=""}
  User.SmokeInitialColor {type=LinearColor, description="", defaultValue {struct=LinearColor, value=<rgba 0.7326 0.7047 0.7047 0.25>}}
  User.SmokeInitialColorScale {type=NiagaraFloat, defaultValue=0.6, description=""}
  User.SmokeScale {type=NiagaraFloat, defaultValue=1, description=""}
  User.SparkAmountScale {type=NiagaraFloat, defaultValue=1, description=""}
 ]
}
